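# Looping ambient sound cue for the LimeRock countdown tree's pink energy glow during pre-event stage 02.
# A single looping wave player plays the pink energy glow loop at boosted volume through the outdoor ambient sound class.
# A stage-specific attenuation asset sets its falloff, and concurrency overrides limit it to a few instances per owner with a short retrigger time.
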SoundCue /LimeRock_CountdownTimer/Assets/Audio/SC_LimeRock_Tree_PreEvent_Stage02_LP
root=WavePlayer_0
nodes [{id=WavePlayer_0, properties {bLooping=true, soundWaveAssetPtr=&/LimeRock_Tree_PinkEnergyGlow_LP}}]
properties {
 attenuationSettings=&/ATT_LimeRock_Tree_PreEvent_Stage02
 maxDistance=10800
 soundClassObject=&Fort_Ambient_Outdoors_Athena
 volumeMultiplier=2.2
 concurrencyOverrides {bLimitToOwner=true, maxCount=2, retriggerTime=0.25}
 concurrencySet [null]
}
